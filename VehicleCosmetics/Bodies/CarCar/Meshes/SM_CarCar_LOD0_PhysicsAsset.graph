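# Physics asset for the LOD0 CarCar vehicle body mesh in the vehicle cosmetics set.
# It holds a single body on the frame bone, approximated by one rotated capsule offset from the bone origin.
# It defines no constraints, so it has no articulated parts.
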PhysicsAsset /VehicleCosmetics/Bodies/CarCar/Meshes/SM_CarCar_LOD0_PhysicsAsset
bodies {frame [{shapeType=Capsule, length=74.4372, radius=39.5315, center=<-13.1629 12.1439 -4.0521>, rotation=<rot 0 85.4081 -73.2861>}]}
constraints []
